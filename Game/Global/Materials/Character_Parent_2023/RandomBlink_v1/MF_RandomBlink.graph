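# Material function that outputs a masked, periodic blink pulse with phase randomized by UV-based noise.
# Time scaled by BlinkFrequency is offset by RandomOffset and the red plus green channels of a vector noise sampled at the UVs.
# Frac, bias-scale and abs turn that into a triangle wave, which a power sharpens before the Mask multiply.
# The exponent is BlinkSpeed scaled up and divided by frequency, keeping snappiness consistent across blink rates.
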MaterialFunction /Game/Global/Materials/Character_Parent_2023/RandomBlink_v1/MF_RandomBlink
graph {nodes [
  n1 = Time()
  n2 = FunctionInput() {inputName=BlinkFrequency, inputType=FunctionInput_Scalar, sortPriority=2, previewValue=<1 0 0 1>}
  n3 = FunctionInput() {inputName=RandomOffset, inputType=FunctionInput_Scalar, sortPriority=1}
  n4 = VectorNoise(World Position=AppendVector(A=FunctionInput() {inputName=UVs, inputType=FunctionInput_Vector2, sortPriority=1}, B=Constant()))
  n5 = Add(A=ComponentMask(None=@n4) {r=true}, B=ComponentMask(None=@n4) {g=true})
  n6 = Frac(None=Add(A=Multiply(A=@n1, B=@n2), B=Add(A=@n3, B=@n5)))
  n7 = OneMinus(None=Abs(None=ConstantBiasScale(None=@n6) {bias=-0.5, desc=-1 - 1, scale=2}))
  n8 = Divide(
   A=Multiply(A=FunctionInput() {inputName=BlinkSpeed, inputType=FunctionInput_Scalar, sortPriority=3, previewValue=<1 0 0 1>}) {constB=10}
   B=NamedRerouteUsage()
  ) {desc=Adjust snappiness based on frequency to keep consistent}
  FunctionOutput(None=Multiply(A=Power(Base=@n7, Exp=@n8), B=FunctionInput() {inputName=Mask, inputType=FunctionInput_Scalar, previewValue=<1 0 0 1>}))
  NamedRerouteDeclaration(None=@n2) {name=Freq, nodeColor=<rgba 1 0 0.9609 1>}
 ]}
